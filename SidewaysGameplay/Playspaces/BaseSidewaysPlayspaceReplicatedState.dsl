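; Base replicated-state actor for a Sideways playspace that tracks whether the local player is inside its owning playspace.
; On remote clients, a repeating timer brute-force polls membership and calls enter or exit handlers on changes, a stopgap marked for replacement.
; Both handlers are empty here.
; An optional debug timer draws a sphere on the authority showing the net cull distance.
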(event EventBeginPlay
  (if (Variables|Default|GetbDebugRelevancyRange)
    (Utilities|Time|SetTimerbyEvent (delegate Custom|CustomEvent_0) 0.1 true)
    (else
      _))
  ;[1 0 0.004128] REPLACE ME -- Brute force check if player is in playspace and call handler functions
  (Networking|SwitchHasAuthority
    (:Remote
      (Utilities|Time|SetTimerbyEvent (delegate Custom|CheckForLocalPlayer) 0.25 true))))

(event Custom|CustomEvent_0
  (Networking|SwitchHasAuthority
    (:Authority
      (Rendering|Debug|DrawDebugSphere (Transformation|GetActorLocation) (Math|Float|Sqrt (Variables|Replication|GetNetCullDistanceSquared)) 12 "(R=1.000000,G=1.000000,B=1.000000,A=1.000000)" 0.2))))

(event Custom|CheckForLocalPlayer
  (bind _blocalplayerinplayspace (Variables|Default|GetbLocalPlayerInPlayspace))
  ;[1 0 0.004128] REPLACE ME -- Brute force check if player is in playspace and call handler functions
  (if (CallFunction|IsLocalPlayerinOwningPlayspace 0)
    (if (not _blocalplayerinplayspace)
      (Variables|Default|SetbLocalPlayerInPlayspace true)
      (CallFunction|HandleLocalPlayerEnteredPlayspace))
    (else
      (if _blocalplayerinplayspace
        (Variables|Default|SetbLocalPlayerInPlayspace false)
        (CallFunction|HandleLocalPlayerExitedPlayspace)))))

(event Collision|EventActorBeginOverlap
  (params (OtherActor :type Actor%20Object%20Reference)))

(event EventTick
  (params (DeltaSeconds :type Float%20%28single-precision%29)))

(fn HandleLocalPlayerEnteredPlayspace)

(fn HandleLocalPlayerExitedPlayspace)

(fn ConstructionScript)
