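; SmallSidewaysPlayspace is a Playspace Blueprint in SidewaysGameplay that logs users joining and leaving it as debug messages.
; On BeginPlay it binds handlers to the playspace's user-added and user-removed notifications.
; Each handler breaks the user struct, converts its unique ID to a debug string and prints it with an added or removed label.
; Tick and the construction script are empty, so the graph holds no other gameplay logic.
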(event EventBeginPlay
  (Playspace|AssignNotifyPlayspaceUserAdded (delegate Custom|NotifyPlayspaceUserAdded_Event_0))
  (Playspace|AssignNotifyPlayspaceUserRemoved (delegate Custom|NotifyPlayspaceUserRemoved_Event_0)))

(event EventTick
  (params (DeltaSeconds :type Float%20%28single-precision%29)))

(event Custom|NotifyPlayspaceUserAdded_Event_0
  (params (AddedUser :type Playspace%20User%20Structure%20%28by%20ref%29))
  (bind _returnvalue (PlayerInfo|GetDebugStringforUniqueId (Utilities|Struct|BreakPlayspaceUser AddedUser)))
  (Development|PrintString (Utilities|String|Append "User Added: " _returnvalue)))

(event Custom|NotifyPlayspaceUserRemoved_Event_0
  (params (RemovedUser :type Playspace%20User%20Structure%20%28by%20ref%29))
  (bind _returnvalue (PlayerInfo|GetDebugStringforUniqueId (Utilities|Struct|BreakPlayspaceUser RemovedUser)))
  (Development|PrintString (Utilities|String|Append "User Removed: " _returnvalue)))

(fn ConstructionScript)
