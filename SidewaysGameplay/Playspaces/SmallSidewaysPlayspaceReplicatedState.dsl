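; Replicated state Blueprint for the small Sideways playspace, which currently only prints debug messages.
; On non-authority machines, a RepParamHelper change event switches on the gameplay tag and prints the new overkill flag, encounter state or maximum wave duration.
; Local player enter and exit overrides call their parent, then print a message.
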(event OnReplicatedParamChanged(RepParamHelper)
  (params (ParamTag :type Gameplay%20Tag%20Structure%20%28by%20ref%29))
  (if (not (Networking|HasAuthority))
    (switch Utilities|FlowControl|Switch|SwitchonGameplayTag ParamTag
      (:RepParam.Bool.Sideways.OverkillEnabled
        (Development|PrintString (Utilities|String|Append "OVERKILL CHANGED: " (Utilities|String|ToString(Boolean) (Class|ActorComponentReplicatedParamHelper|GetParamasBool (Variables|Default|GetRepParamHelper) "(TagName=\"RepParam.Bool.Sideways.OverkillEnabled\")")))))
      (:RepParam.Integer.Sideways.EncounterState
        (Development|PrintString (Utilities|String|Append "ENCOUNTER STATE CHANGED: " (Utilities|String|EnumtoString (Utilities|Enum|BytetoEnumEnum_Athena_SidewaysEncounter_State (Math|Conversions|ToByte(Integer) (Class|ActorComponentReplicatedParamHelper|GetParamasInt (Variables|Default|GetRepParamHelper) "(TagName=\"RepParam.Integer.Sideways.EncounterState\")")))))))
      (:RepParam.Float.Sideways.WaveDuration.Max
        (Development|PrintString (Utilities|String|Append "WAVE DURATION CHANGED: " (Utilities|String|ToString(Float) (Class|ActorComponentReplicatedParamHelper|GetParamasFloat (Variables|Default|GetRepParamHelper) "(TagName=\"RepParam.Float.Sideways.WaveDuration.Max\")"))))))))

(fn HandleLocalPlayerEnteredPlayspace
  (|CallParentFunction)
  (Development|PrintString "LOCAL PLAYER ENTERED PLAYSPACE"))

(fn HandleLocalPlayerExitedPlayspace
  (|CallParentFunction)
  (Development|PrintString "LOCAL PLAYER EXITED PLAYSPACE"))

(fn ConstructionScript)
